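# Physics asset for the Invader Male pickaxe, a two-handed melee weapon cosmetic.
# A long capsule body covers the handle bone, and a small rotated capsule covers the dyn_base bone.
# A tight angular constraint links dyn_base to the handle, allowing only slight swing and twist.
# A second, looser constraint has no bones assigned.
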PhysicsAsset /BRCosmetics/Weapons/FORT_Melee/Pickaxe_Invader_Male/Meshes/Invader_Male_Axe_Physics
bodies {
 Melee_TwoHanded_Handle [{shapeType=Capsule, length=140, radius=3, center=<0 0 33.8593>}]
 dyn_base [{shapeType=Capsule, length=3, radius=3, center=<3.3688 0 0>, rotation=<rot 90 0 0>}]
}
constraints [
 {bone1Name=None, bone2Name=None, swing1LimitDegrees=20, swing2LimitDegrees=20, twistLimitDegrees=20}
 {bone1Name=dyn_base, bone2Name=Melee_TwoHanded_Handle, swing1LimitDegrees=5, swing2LimitDegrees=5, twistLimitDegrees=5}
]
