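# Animation Blueprint for the Invader Male pickaxe skeleton, adding secondary physics motion to the melee weapon.
# A sequence player feeds the axe animation into a RigidBody node that simulates an override physics asset relative to the two-handed handle bone, with scaled component motion and overridden gravity, then outputs the pose.
# An Anim Dynamics spring node for the dyn_hose bone is left unconnected, and the event graph has no logic.
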AnimBlueprint /BRCosmetics/Weapons/FORT_Melee/Pickaxe_Invader_Male/Meshes/Invader_Male_Axe_Skeleton_AnimBP
target_skeleton=&/Invader_Male_Axe_Skeleton
graphs [
 AnimGraph {nodes [
   Anim Dynamics - Bone: dyn_hose(Alpha=1, ExternalForce="0,0,0") {
    angularSpringConstant=400
    bAngularSpring=true
    bGravityOverrideInSimSpace=true
    bLinearSpring=true
    gravityScale=0.5
    linearSpringConstant=400
    alphaBoolBlend {blendOption=Linear}
    boundBone {boneName=dyn_hose}
    componentLinearAccScale=<0.15 0.15 0.15>
    componentLinearVelScale=<0.15 0.15 0.15>
    physicsBodyDefinitions [{
      sphereCollisionRadius=0
      boundBone {boneName=dyn_hose}
      boxExtents=<25 25 25>
      constraintSetup {angularLimitsMax=<20 20 20>, angularLimitsMin=<-20 -20 -20>, angularTarget=<0 0 0>, linearAxesMax=<5 5 5>, linearAxesMin=<-5 -5 -5>}
      localJointOffset=<0 0 0>
     }]
   }
   n1 = RigidBody(ComponentPose=Local To Component(LocalPose=Sequence Player 'Invader_Male_Axe_Pickaxe_Anim'() {sequence=/BRCosmetics/Weapons/FORT_Melee/Pickaxe_Invader_Male/Animation/Invader_Male_Axe_Pickaxe_Anim})) {
    bClampLinearTranslationLimitToRefPose=true
    bOverrideWorldGravity=true
    overridePhysicsAsset=&/Invader_Male_Axe_Physics
    simulationSpace=BaseBoneSpace
    alphaBoolBlend {blendOption=Linear}
    baseBoneRef {boneName=Melee_TwoHanded_Handle}
    componentAppliedLinearAccClamp=<3000 3000 3000>
    componentLinearAccScale=<0.3 0.3 0.4>
    componentLinearVelScale=<0.3 0.3 0.4>
    overrideWorldGravity {z=-980}
   }
   Output Pose(Result=Component To Local(ComponentPose=@n1)) {layerGroup=None}
  ]}
 EventGraph {nodes [
   TryGetPawnOwner()
   Event BlueprintUpdateAnimation()
  ]}
]
variables []
