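; Weapon Blueprint for the Invader pickaxe (Athena) that drives its built-in idle and swing trail particle effects.
; It activates the idle VFX component on equip and deactivates it on local unequip.
; Swing events set a direction flag on the trail component and a swinging flag on the idle component.
; Impact clears the swinging flag and pulses an impacted flag on the idle effect for about two frames, using a delta-time-scaled delay.
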(event EventBeginPlay)

(event EventTick
  (params (DeltaSeconds :type Float%20%28single-precision%29)))

(event FortWeapon|EventOnEquippedAndAttached
  ;[1 1 1] Built In Pickaxe FX Logic
  (|CallParentFunction)
  (Components|Activation|Activate (CallFunction|GetIdleVFXComponent) true))

(event FortWeapon|EventOnLocalUnEquip
  ;[1 1 1] Built In Pickaxe FX Logic
  (|CallParentFunction)
  (Components|Activation|Deactivate (CallFunction|GetIdleVFXComponent)))

(event EventSwingRight
  (|CallParentFunction)
  ;[1 1 1] Built In Pickaxe FX Logic
  (Effects|Components|ParticleSystem|SetBoolParameter (CallFunction|GetAnimTrailVFXComponent) "bIsSwingingLeft" false)
  (Effects|Components|ParticleSystem|SetBoolParameter (CallFunction|GetIdleVFXComponent) "bIsSwinging" true))

(event EventSwingLeft
  (|CallParentFunction)
  ;[1 1 1] Built In Pickaxe FX Logic
  (Effects|Components|ParticleSystem|SetBoolParameter (CallFunction|GetAnimTrailVFXComponent) "bIsSwingingLeft" true)
  (Effects|Components|ParticleSystem|SetBoolParameter (CallFunction|GetIdleVFXComponent) "bIsSwinging" true))

(event EventSwingRightEnd
  (|CallParentFunction)
  ;[1 1 1] Built In Pickaxe FX Logic
  (Effects|Components|ParticleSystem|SetBoolParameter (CallFunction|GetIdleVFXComponent) "bIsSwinging" false))

(event EventSwingLeftEnd
  (|CallParentFunction)
  ;[1 1 1] Built In Pickaxe FX Logic
  (Effects|Components|ParticleSystem|SetBoolParameter (CallFunction|GetIdleVFXComponent) "bIsSwinging" false))

(event FortWeapon|EventOnPlayImpactFX
  (params (HitResult :type Hit%20Result%20Structure%20%28by%20ref%29) (ImpactPhysicalSurface :type EPhysicalSurface%20Enum :default "SurfaceType_Default") (SpawnedPSC :type FXSystem%20Component%20Object%20Reference))
  (bind _returnvalue (CallFunction|GetIdleVFXComponent))
  ;[1 1 1] Both
  (|CallParentFunction HitResult ImpactPhysicalSurface SpawnedPSC)
  (Effects|Components|ParticleSystem|SetBoolParameter _returnvalue "HasImpacted" true)
  (Effects|Components|ParticleSystem|SetBoolParameter _returnvalue "bIsSwinging" false)
  (Utilities|FlowControl|Delay (* (Utilities|Time|GetWorldDeltaSeconds) 2.0))
  (Effects|Components|ParticleSystem|SetBoolParameter _returnvalue "HasImpacted" false))

(fn ConstructionScript
  (|CallParentFunction))
