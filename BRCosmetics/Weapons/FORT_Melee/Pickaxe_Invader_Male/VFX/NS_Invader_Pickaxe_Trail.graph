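# Niagara ribbon trail for the Invader male pickaxe, following two bones or sockets on the user-supplied skeletal mesh.
# The base emitter uses User.TrailColor, teal by default.
# A second emitter, switched by User.UseRainbowTrail against a threshold, picks random colors from a "rainbow" curve that actually spans only orange to red.
# Both spawn short-lived anim-trail particles whose alpha follows a fade curve multiplied by a time-of-day scalar.
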NiagaraSystem /BRCosmetics/Weapons/FORT_Melee/Pickaxe_Invader_Male/VFX/NS_Invader_Pickaxe_Trail
dependencies {
 usedDataInterfaces [&NiagaraDataInterfaceCurve, &NiagaraDataInterfaceColorCurve]
 usedDynamicInputs []
 usedModules [&SystemState]
}
emitters {
 FN_AnimTrail {
  emitterSpawnScript [
   &FN_RibbonIDOffsetUtility()
   /SetVariables(Emitter.RainbowCurve=DataInterface() {propertyValues {
      AlphaCurve {keys [<at 0.4247 1>]}
      BlueCurve {keys [<at 0 0 RCIM_Cubic>,<at 0.3514 0 RCIM_Cubic>,<at 1 0 RCIM_Cubic>]}
      GreenCurve {keys [
        <at 0 0.399 RCIM_Cubic>
        {interpMode=RCIM_Cubic, leaveTangent=-0.6263, time=0.2, value=0.3}
        {interpMode=RCIM_Cubic, leaveTangent=0.125, time=1, value=0}
       ]}
      RedCurve {keys [<at 0 1 RCIM_Cubic>,<at 0.3514 1 RCIM_Cubic>,<at 1 1 RCIM_Cubic>]}
     }}) {isSetParameters=true}
  ]
  emitterUpdateScript [
   &EmitterState()
   &FN_SpawnAnimTrail(Enable:NiagaraBool=&SetBoolByFloatComparison(A=$User.UseRainbowTrail, B=0.5))
  ]
  particleSpawnScript [
   &FN_InitializeAnimTrailBoneIndecies(First Bone or Socket Index=1, Second Bone or Socket Index=0)
   &FN_InitializeAnimTrail(
    AnimTrailWidth=0.75
    Color=&ColorFromCurve(ColorCurve=$Emitter.RainbowCurve, CurveIndex=&RandomRangeFloat(Minimum=0, Maximum=1))
    Lifetime=0.15
   )
  ]
  particleUpdateScript [
   &FN_UpdateAnimTrail(Locked Particle RibbonWidth=1, SkeletalMesh=$User.SkeletalMesh) {module=FN_UpdateAnimTrail001}
   &ScaleColor(
    Scale Alpha=&Multiply_Float(
     A=&FloatFromCurve(
      FloatCurve=DataInterface() {propertyValues {Curve {keys [{arriveTangent=19.9617, interpMode=RCIM_Cubic, tangentMode=RCTM_User, time=0, value=1}, <at 1 0>]}}}
      Scale Curve=0.35
     )
     B=&FNTimeOfDayScalar(Night Scale=0.3, Morning Scale=0.5, Evening Scale=0.5)
    )
    ScaleA=true
    ScaleRGB=true
   )
  ]
  rendererClasses [&NiagaraRibbonRendererProperties]
  renderers [{rendererClass=&NiagaraRibbonRendererProperties}]
 }
 FN_AnimTrail001 {
  emitterSpawnScript [&FN_RibbonIDOffsetUtility()]
  emitterUpdateScript [
   &EmitterState()
   &FN_SpawnAnimTrail()
  ]
  particleSpawnScript [
   &FN_InitializeAnimTrailBoneIndecies(First Bone or Socket Index=1, Second Bone or Socket Index=0)
   &FN_InitializeAnimTrail(Color=$User.TrailColor, Lifetime=0.15)
  ]
  particleUpdateScript [
   &FN_UpdateAnimTrail(Locked Particle RibbonWidth=1, SkeletalMesh=$User.SkeletalMesh) {module=FN_UpdateAnimTrail001}
   &ScaleColor(
    Scale Alpha=&Multiply_Float(
     A=&FloatFromCurve(
      FloatCurve=DataInterface() {propertyValues {Curve {keys [{arriveTangent=19.9617, interpMode=RCIM_Cubic, tangentMode=RCTM_User, time=0, value=1}, <at 1 0>]}}}
      Scale Curve=0.35
     )
     B=&FNTimeOfDayScalar(Night Scale=0.3, Morning Scale=0.5, Evening Scale=0.5)
    )
    ScaleA=true
    ScaleRGB=true
   )
  ]
  rendererClasses [&NiagaraRibbonRendererProperties]
  renderers [{rendererClass=&NiagaraRibbonRendererProperties}]
 }
}
summary {
 systemName=NS_Invader_Pickaxe_Trail
 emitters [
  {emitterName=FN_AnimTrail001, rendererClasses [&NiagaraRibbonRendererProperties]}
  {emitterName=FN_AnimTrail, rendererClasses [&NiagaraRibbonRendererProperties]}
 ]
 userVariables [
  User.RibbonIDOffset {type=NiagaraInt32}
  User.SkeletalMesh {
   type=NiagaraDataInterfaceSkeletalMesh
   defaultValue {
    struct=NiagaraExt_VariableValue_DataInterface
    value {dataInterface=&self:NiagaraDataInterfaceSkeletalMesh_0}
   }
  }
  User.TrailColor {type=LinearColor, description="", defaultValue {struct=LinearColor, value=<rgba 0.063 0.4999 0.5149 0.75>}}
  User.UseRainbowTrail {type=NiagaraFloat, defaultValue=0, description=""}
 ]
}
